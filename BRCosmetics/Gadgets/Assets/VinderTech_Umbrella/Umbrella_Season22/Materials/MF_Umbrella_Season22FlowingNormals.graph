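# Material function for the Season 22 umbrella that adds animated flowing liquid normals to incoming material attributes, with a separate LiquidNormal output.
# Scrolling noise normals use radial coordinates from a UV texture, with explicit derivatives, and switch to linearly panned noise on masked seams.
# These mix with a flowmap normal parameter and are flattened by seam mask.
# An angle-corrected blend, alpha and channel-mask strength controls shape the normals.
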MaterialFunction /BRCosmetics/Gadgets/Assets/VinderTech_Umbrella/Umbrella_Season22/Materials/MF_Umbrella_Season22FlowingNormals
graph {nodes [
  FunctionOutput(None=SetMaterialAttributes!(MaterialAttributes=NamedRerouteUsage(), Normal=NamedRerouteUsage()) {attributeSetTypes [#1]})
  FunctionOutput(None=NamedRerouteUsage()) {outputName=LiquidNormal}
  n1 = NamedRerouteUsage()
  n2 = ScalarParameter() {group=LiquidNormal, parameterName=LQN_NormalStrengthMin}
  n3 = ScalarParameter() {defaultValue=1, group=LiquidNormal, parameterName=LQN_NormalStrengthMax}
  n4 = FunctionInput() {bUsePreviewValueAsDefault=true, inputName=LiquidNormalMask, inputType=FunctionInput_Vector4, previewValue=<1 1 1 1>}
  n5 = ChannelMaskParameter(None=NamedRerouteDeclaration(None=@n4) {name=LiquidNormalMask, nodeColor=<rgba 0.5 0 1 1>}) {group=LiquidNormal, parameterName=LQN_MaskChannel}
  NamedRerouteDeclaration(None=LinearInterpolate(
    A=Constant3Vector() {constant=<rgba 0 0 1 0>}
    B=NamedRerouteDeclaration(None=@n1) {name=LiquidNormal_OUT, nodeColor=<rgba 0 1 0.5313 1>}
    Alpha=LinearInterpolate(A=@n2, B=@n3, Alpha=@n5.R)
   )) {name=Normal_OUT, nodeColor=<rgba 0.0078 0 1 1>}
  n6 = GetMaterialAttributes(None=FunctionInput() {inputType=FunctionInput_MaterialAttributes}) {attributeGetTypes [#1]}
  NamedRerouteDeclaration(None=@n6.Normal) {name=Normal_IN, nodeColor=<rgba 1 0 0.1406 1>}
  n7 = Add(A=ComponentMask(None=TextureSample() {samplerType=SAMPLERTYPE_Normal, texture=/BRCosmetics/Gadgets/Assets/VinderTech_Umbrella/Umbrella_Season22/Textures/T_UV_UmbrellaS22}.RGB) {g=true, r=true})
  n8 = MaterialFunctionCall(Vector or UVs=Add(A=Divide(A=@n7)) {constB=-0.505}) {materialFunction=&VectorToRadialValue}
  n9 = AppendVector(
   A=ScalarParameter() {defaultValue=1, group=LiquidNormal, parameterName=Radial_Tiling_X}
   B=ScalarParameter() {defaultValue=0.7, group=LiquidNormal, parameterName=Radial_Tiling_Y}
  )
  n10 = Time()
  n11 = Multiply(A=@n10, B=ScalarParameter() {defaultValue=-0.15, group=LiquidNormal, parameterName=Radial_SpeedUV_R}) {constB=-0.05}
  n12 = Multiply(A=@n10, B=ScalarParameter() {defaultValue=-0.1, group=LiquidNormal, parameterName=Radial_SpeedUV_G}) {constB=0.1}
  n13 = Multiply(A=Add(A=Multiply(A=@n8.Radial Coordinates, B=@n9), B=AppendVector(A=@n11, B=@n12))) {constB=2}
  n14 = TextureCoordinate()
  n15 = DDX(Value=@n14)
  n16 = DDY(Value=@n14)
  n17 = Multiply(A=Panner(Time=Time()) {speedY=0.05}) {constB=5}
  n18 = StaticComponentMaskParameter(None=TextureSampleParameter2D() {group=LiquidNormal, parameterName=SeamMask, texture=&T_BlacksRGB}.RGB) {defaultG=true, group=LiquidNormal, parameterName=SeamMaskChannel}
  NamedRerouteDeclaration(None=LinearInterpolate(
    A=TextureSample!(UVs=@n13, Tex, "DDX(UVs)"=@n15, "DDY(UVs)"=@n16, Apply View MipBias) {mipValueMode=TMVM_Derivative, samplerType=SAMPLERTYPE_Normal, texture=/Game/Global/Textures/Default/Noise/T_FlowWarp_Noise}.RGB
    B=TextureSample!(UVs=@n17, Tex, Apply View MipBias) {samplerType=SAMPLERTYPE_Normal, texture=&T_FlowWarp_Noise}.RGB
    Alpha=@n18
   )) {name=new warp uv, nodeColor=<rgba 0.2656 1 0 1>}
  n19 = FunctionInput() {bUsePreviewValueAsDefault=true, inputName=Alpha, inputType=FunctionInput_Scalar, previewValue=<1 0 0 1>}
  NamedRerouteDeclaration(None=@n19) {name=Alpha_IN, nodeColor=<rgba 1 0 0.9844 1>}
  n20 = NamedRerouteUsage()
  n21 = NamedRerouteUsage()
  NamedRerouteDeclaration(None=LinearInterpolate(
    A=@n20
    B=MaterialFunctionCall(BaseNormal=@n20, AdditionalNormal=@n21) {materialFunction=&BlendAngleCorrectedNormals}.Result
    Alpha=NamedRerouteUsage()
   )) {name=Normal, nodeColor=<rgba 1 0.0703 0 1>}
  NamedRerouteDeclaration(None=@n6.MaterialAttributes) {name=Material_IN, nodeColor=<rgba 0 1 0.5547 1>}
  NamedRerouteDeclaration(None=VertexInterpolator(VS=PreSkinnedNormal()).PS) {name=PS_Normal, nodeColor=<rgba 1 0 0.8437 1>}
  NamedRerouteDeclaration(None=VertexInterpolator(VS=PreSkinnedPosition()).PS) {name=PS_Position, nodeColor=<rgba 1 0 0.1641 1>}
  n22 = TextureSampleParameter2D() {group=LiquidNormal, parameterName="flowmapNM ", samplerType=SAMPLERTYPE_Normal, texture=&FlatNormalBlue}
  n23 = NamedRerouteUsage()
  n24 = Saturate(None=Add(A=@n18, B=ScalarParameter() {defaultValue=0.1, group=LiquidNormal, parameterName=Flowmap_indensity}))
  n25 = ScalarParameter() {defaultValue=0.8, group=LiquidNormal, parameterName=SeamFlattenNormal}
  NamedRerouteDeclaration(None=MaterialFunctionCall(Normal=LinearInterpolate(A=@n22.RGB, B=@n23, Alpha=@n24), Flatness=LinearInterpolate(B=@n25, Alpha=@n18) {constB=0.8}) {materialFunction=&FlattenNormal}.Result) {name=LiquidNormal, nodeColor=<rgba 1 0.7266 0 1>}
  ScalarParameter()
 ]}
