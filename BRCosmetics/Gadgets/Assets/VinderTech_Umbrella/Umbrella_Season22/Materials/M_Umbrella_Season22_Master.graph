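# Master character material for the Season 22 umbrella cosmetic, giving it an iridescent chrome look with flowing liquid normals.
# PreFX attributes pass through a flowing-normals function, masked by a sphere in pre-skinned local space, before attribute overrides and PostFX.
# A color layer sums Fresnel-indexed chrome curve atlas samples, Fresnel-driven hue shift and a color-dodged, desaturated diffuse.
# Base color blends two rerouted colors, with near-full metallic, low roughness and slight anisotropy.
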Material /BRCosmetics/Gadgets/Assets/VinderTech_Umbrella/Umbrella_Season22/Materials/M_Umbrella_Season22_Master
graph {
 properties {
  bUseMaterialAttributes=true
  bUsedWithMeshDeformer=true
  bUsedWithSkeletalMesh=true
  pixelDepthOffsetMode=PDOM_Legacy
  usedShadingModels=Default Lit
 }
 nodes [
  n1 = MaterialFunctionCall() {materialFunction=&BaseFN_CharacterFunction_PreFX}
  n2 = Constant3Vector()
  n3 = ScalarParameter() {defaultValue=20, group=PickaxeBites, parameterName=liquidNormalRadius}
  n4 = ScalarParameter() {defaultValue=1, group=PickaxeBites, parameterName=liquidNormalHardness}
  n5 = GetMaterialAttributes(None=MaterialFunctionCall!(
    Alpha=SphereMask(A=@n1.Pre skinned Local Position, B=@n2, Radius=@n3, Hardness=@n4) {attenuationRadius=20, hardnessPercent=1}
    LiquidNormalMask=Constant4Vector() {constant=<rgba 1 1 1 1>}.RGBA
    In=@n1.Result
   ) {materialFunction=&/MF_Umbrella_Season22FlowingNormals}.Result) {attributeGetTypes [#1]}
  n6 = SetMaterialAttributes!(
   MaterialAttributes=@n5.MaterialAttributes
   Emissive Color=NamedRerouteUsage()
   Normal=NamedRerouteUsage()
   Base Color=LinearInterpolate(A=NamedRerouteUsage(), B=NamedRerouteUsage()) {constAlpha=0.6}
   Metallic=Constant() {r=0.965}
   Specular=NamedRerouteUsage()
   Roughness=Constant() {r=0.23}
   Anisotropy=Constant() {r=0.03}
  ) {attributeSetTypes [
    #2
    #1
    #3
    #4
    #5
    #6
    #7
   ]}
  MaterialFunctionCall(AO=@n1.AO, Distance Blend=@n1.AO, Fresnel=@n1.AO, Gmap=@n1.AO, Material=@n6) {materialFunction=&BaseFN_CharacterFunction_PostFX}
  NamedRerouteDeclaration(None=@n5.Normal) {name=Normalfinal, nodeColor=<rgba 0.2187 1 0 1>}
  n7 = GetMaterialAttributes(None=@n1.Result) {attributeGetTypes [
    #3
    #6
    #2
    #5
   ]}
  NamedRerouteDeclaration(None=@n7.Base Color) {name=baseColor, nodeColor=<rgba 0 0.8828 1 1>}
  n8 = Constant() {r=2}
  n9 = Transform(None=NamedRerouteUsage())
  n10 = CurveAtlasRowParameter(CurveTime=Frac(None=MaterialFunctionCall(Exponent=@n8, Vertex / Pixel Normal=@n9) {materialFunction=&MF_Fresnel_FX}.Fresnel)) {
   atlas=&/C_UmbrellaChrome
   curve=&/C_UmbrellaChrome_00
   parameterName=Chrome
  }
  n11 = Constant() {r=2}
  n12 = PixelNormalWS()
  n13 = CurveAtlasRowParameter(CurveTime=Frac(None=MaterialFunctionCall(Exponent=@n11, Vertex / Pixel Normal=@n12) {materialFunction=&MF_Fresnel_FX}.Fresnel)) {
   atlas=&C_UmbrellaChrome
   curve=&C_UmbrellaChrome_00
   parameterName=Chrome
  }
  n14 = MaterialFunctionCall(Exponent=ScalarParameter() {defaultValue=2.5, parameterName=EdgeNormal_FresnelExponent}, Vertex / Pixel Normal=VertexNormalWS()) {materialFunction=&MF_Fresnel_FX}
  n15 = ScalarParameter() {defaultValue=0.15, parameterName=FresnelHueShiftExponent}
  n16 = Transform(None=NamedRerouteUsage())
  n17 = Frac(None=Multiply(
    A=MaterialFunctionCall(Exponent=@n15, Vertex / Pixel Normal=@n16) {materialFunction=&MF_Fresnel_FX}.Fresnel
    B=ScalarParameter() {defaultValue=1, parameterName=HueShiftScale}
   ))
  n18 = VectorParameter() {parameterName=AlbedoHueShift, defaultValue=<rgba 0.85 0.9028 0.9 0>}
  n19 = Add(
   A=Multiply(A=Add(A=@n10, B=@n13), B=LinearInterpolate(Alpha=@n14.Fresnel Inverted) {constA=0.1})
   B=Multiply(
    A=VectorParameter() {parameterName=AlbedoScale, defaultValue=<rgba 0.75 0.6437 0.6136 0>}.RGB
    B=MaterialFunctionCall(Hue Shift Percentage=@n17, Texture=@n18.RGB) {materialFunction=&HueShift}.Result
   )
  )
  n20 = Desaturation(None=NamedRerouteUsage(), Fraction=ScalarParameter() {defaultValue=0.92, parameterName=DiffuseDesaturation})
  n21 = Add(
   A=Divide(A=VectorParameter() {parameterName=Diffuse_ColorDodgeBase, defaultValue=<rgba 0.05 0.0427 0.0375 0>}.RGB, B=OneMinus(None=@n20))
   B=ScalarParameter() {defaultValue=-0.5, parameterName=DiffuseBias}
  )
  n22 = ScalarParameter() {defaultValue=1.6, parameterName=DiffuseScale}
  n23 = NamedRerouteDeclaration(None=Saturate(None=Multiply(A=@n21, B=@n22))) {name=Diffuse_1, nodeColor=<rgba 1 0 0.1172 1>}
  NamedRerouteDeclaration(None=Desaturation(None=Add(A=@n19, B=@n23), Fraction=ScalarParameter() {defaultValue=-1, parameterName=Fraction})) {name=pink color, nodeColor=<rgba 1 0.5625 0 1>}
  NamedRerouteDeclaration(None=@n7.Emissive Color) {name=Emissive, nodeColor=<rgba 0 1 0.1562 1>}
  NamedRerouteDeclaration(None=@n7.Specular) {name=Specular, nodeColor=<rgba 1 0.1406 0 1>}
 ]
}
